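# Animation Blueprint for the Star Walker pickaxe that makes its four shard bones float and wobble procedurally.
# Each update, the EventGraph builds four offset vectors from sine and cosine waves of game time at differing frequencies and amplitudes.
# Two of these vectors, scaled up, also become rotators for the left and right shards.
# The AnimGraph chains Modify Bone nodes after the input pose, applying the offsets additively in bone space.
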AnimBlueprint /BRCosmetics/Weapons/FORT_Melee/Pickaxe_Star_Walker/Meshes/Star_Walker_Axe_AnimBP
target_skeleton=&/Star_Walker_Axe_Skeleton
graphs [
 AnimGraph {nodes [
   AnimGraphNode_SubInput_0 = Input Pose()
   n1 = "Transform (Modify) Bone - Bone: dyn_rightShard"(
    ComponentPose=Local To Component(LocalPose=@AnimGraphNode_SubInput_0)
    Alpha=1
    Translation=Get ShardFloatA() -> (ShardFloatA:Vector)
    Rotation=Get ShardRotD() -> (ShardRotD:Rotator)
    Scale="1,1,1"
   ) {
    rotationMode=BMM_Additive
    rotationSpace=BCS_BoneSpace
    translationMode=BMM_Additive
    translationSpace=BCS_BoneSpace
    alphaBoolBlend {blendOption=Linear}
    boneToModify {boneName=dyn_rightShard}
   }
   n2 = "Transform (Modify) Bone - Bone: dyn_leftShard"(
    ComponentPose=@n1
    Alpha=1
    Translation=Get ShardFloatB() -> (ShardFloatB:Vector)
    Rotation=Get ShardRotC() -> (ShardRotC:Rotator)
    Scale="1,1,1"
   ) {
    rotationMode=BMM_Additive
    rotationSpace=BCS_BoneSpace
    translationMode=BMM_Additive
    translationSpace=BCS_BoneSpace
    alphaBoolBlend {blendOption=Linear}
    boneToModify {boneName=dyn_leftShard}
   }
   n3 = "Transform (Modify) Bone - Bone: dyn_topShard"(ComponentPose=@n2, Alpha=1, Translation=Get ShardFloatC() -> (ShardFloatC:Vector), Rotation="0,0,0", Scale="1,1,1") {translationMode=BMM_Additive, translationSpace=BCS_BoneSpace, alphaBoolBlend {blendOption=Linear}, boneToModify {boneName=dyn_topShard}}
   n4 = "Transform (Modify) Bone - Bone: dyn_bottomShard"(ComponentPose=@n3, Alpha=1, Translation=Get ShardFloatD() -> (ShardFloatD:Vector), Rotation="0,0,0", Scale="1,1,1") {translationMode=BMM_Additive, translationSpace=BCS_BoneSpace, alphaBoolBlend {blendOption=Linear}, boneToModify {boneName=dyn_bottomShard}}
   Output Pose(Result=Component To Local(ComponentPose=@n4)) {layerGroup=None}
  ]}
 EventGraph {nodes [
   Event BlueprintUpdateAnimation()
   n1 = GetGameTimeInSeconds()
   n2 = "Sin (Radians)"(A=float * float(A=float * float(A=1, B=1), B=@n1))
   n3 = "Sin (Radians)"(A=float * float(A=float * float(A=1, B=2), B=@n1))
   n4 = "Sin (Radians)"(A=float * float(A=float * float(A=1, B=1.5), B=@n1))
   > Set ShardFloatA(ShardFloatA:Vector=MakeVector(X=float * float(A=@n4, B=0.3), Y=float * float(A=@n2, B=0.56), Z=float * float(A=@n3, B=0.22))) -> (Output_Get:Vector)
   n5 = "Cos (Radians)"!(A:double=float * float(A=float * float(A=1, B=1.5), B=@n1)) -> (ReturnValue:double)
   n6 = "Cos (Radians)"(A=float * float(A=float * float(A=1, B=1.78), B=@n1))
   n7 = "Cos (Radians)"(A=float * float(A=float * float(A=1, B=1.15), B=@n1))
   > Set ShardFloatB(ShardFloatB:Vector=MakeVector(X=float * float(A=@n5, B=0.6), Y=float * float(A=@n6, B=0.725), Z=float * float(A=@n7, B=0.3))) -> (Output_Get:Vector)
   n8 = float * float(A=float * float(A=1, B=1.5), B=@n1)
   n9 = float * float(A=float * float(A=1, B=1), B=@n1)
   n10 = float * float(A=float * float(A=1, B=2), B=@n1)
   n11 = MakeVector(
    X=float * float(A="Sin (Radians)"(A=@n8), B=0.3)
    Y=float * float(A="Sin (Radians)"(A=@n9), B=0.6)
    Z=float * float(A="Sin (Radians)"(A=@n10), B=0.25)
   )
   > Set ShardFloatC(ShardFloatC:Vector=@n11) -> (Output_Get:Vector)
   n12 = float * float(A=float * float(A=0.6678, B=1.5), B=@n1)
   n13 = float * float(A=float * float(A=1, B=1.47), B=@n1)
   n14 = float * float(A=float * float(A=1, B=0.953), B=@n1)
   n15 = MakeVector(
    X=float * float(A="Sin (Radians)"(A=@n12), B=0.85)
    Y=float * float(A="Sin (Radians)"(A=@n14), B=0.1)
    Z=float * float(A="Sin (Radians)"(A=@n13), B=0.7)
   )
   > Set ShardFloatD(ShardFloatD:Vector=@n15) -> (Output_Get:Vector)
   n16 = BreakVector(InVec=vector * float(A=@n15, B=5))
   > Set ShardRotD(ShardRotD:Rotator=MakeRotator(Roll=@n16, Pitch=@n16.Y, Yaw=@n16.Z)) -> (Output_Get:Rotator)
   n17 = BreakVector(InVec=vector * float(A=@n11, B=7))
   > Set ShardRotC(ShardRotC:Rotator=MakeRotator(Roll=@n17, Pitch=@n17.Y, Yaw=@n17.Z)) -> (Output_Get:Rotator)
   TryGetPawnOwner()
  ]}
]
variables [
 ShardFloatA {type=Vector}
 ShardFloatB {type=Vector}
 ShardFloatC {type=Vector}
 ShardFloatD {type=Vector}
 ShardRotC {type=Rotator}
 ShardRotD {type=Rotator}
]
